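; Master countdown display Blueprint for the Figment S09 event, with a screen, a floating hologram and a skybeam.
; Its construction script applies material overrides, positions the hologram, stretches the beam to its height, and sets visibility and HLOD exclusion.
; It creates dynamic material instances for digit layout, color, textures and scale, and widens hologram bounds.
; Debug flags write phase parameters to the shared MPC_Countdown collection; BeginPlay, Tick and overlap events are empty.
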(event EventBeginPlay)

(event Collision|EventActorBeginOverlap
  (params (OtherActor :type Actor%20Object%20Reference)))

(event EventTick
  (params (DeltaSeconds :type Float%20%28single-precision%29)))

(fn ConstructionScript
  (bind _hologramheight (Variables|Hologram|GetHologramHeight))
  (bind _returnvalue (Math|Vector|MakeVector 0.0 0.0 _hologramheight))
  (bind _hologramskybeam (Variables|Default|GetHologramSkybeam))
  (bind _hologramcountdown (Variables|Default|GetHologramCountdown))
  (bind _hologramskybeam_2 (Variables|Default|GetHologramSkybeam))
  (bind _hologramcountdown_2 (Variables|Default|GetHologramCountdown))
  (bind _countdownscreen (Variables|Default|GetCountdownScreen))
  (bind _countdownhologram_override (Variables|MaterialOverrides|GetCountdownHologram_Override))
  (Utilities|FlowControl|Sequence
    (:then_0
      ;[1 1 1] OverideMaterials if needed
      (Utilities|IsValid (Variables|MaterialOverrides|GetCountdownScreen_Override)
        (:Is%20Valid
          (Rendering|Material|SetMaterial (Variables|Default|GetCountdownScreen) 0 (Variables|MaterialOverrides|GetCountdownScreen_Override)))))
    (:then_1
      (Utilities|IsValid (Variables|MaterialOverrides|GetBeam_Override)
        (:Is%20Valid
          (Rendering|Material|SetMaterial (Variables|Default|GetHologramSkybeam) 0 (Variables|MaterialOverrides|GetBeam_Override)))))
    (:then_2
      (Utilities|IsValid (Variables|MaterialOverrides|GetCountdownHologram_Override)
        (:Is%20Valid
          (Rendering|Material|SetMaterial _hologramcountdown_2 0 _countdownhologram_override)
          (Rendering|Material|SetMaterial _hologramcountdown_2 1 _countdownhologram_override))))
    (:then_3
      ;[1 1 1] Set Height of Hologram
      (Transformation|SetRelativeLocation (Variables|Default|GetHologramCountdown) _returnvalue false false)
      (Transformation|SetRelativeLocation _hologramskybeam (/ _returnvalue 2.0) false false)
      (Transformation|SetRelativeScale3D _hologramskybeam (Math|Vector|MakeVector 1.0 1.0 (/ (.z _returnvalue) 1.2e+02))))
    (:then_4
      (Utilities|FlowControl|Sequence
        (:then_0
          ;[1 1 1] SetVisiblity of elements
          (Rendering|SetVisibility (Variables|Default|GetCountdownScreen) (Variables|BaseSetup|GetScreenVisible) false)
          (if (and (Variables|BaseSetup|GetShowScreenInHLOD) (Variables|BaseSetup|GetScreenVisible))
            (if (Variables|BaseSetup|GetExcludefromMediumHLOD)
              (HLOD|SetExcludedFromHLODLevel _countdownscreen "HLOD0" true)
              (else
                (HLOD|SetExcludedFromHLODLevel _countdownscreen "HLOD0" false)
                (HLOD|SetExcludedFromHLODLevel _countdownscreen "HLOD1" false)))
            (else
              (HLOD|SetExcludedFromHLODLevel _countdownscreen "HLOD0" true)
              (HLOD|SetExcludedFromHLODLevel _countdownscreen "HLOD1" true))))
        (:then_1
          (Rendering|SetVisibility (Variables|Default|GetHologramCountdown) (Variables|BaseSetup|GetHologramVisible) false)
          (if (and (Variables|BaseSetup|GetShowHologramInHLOD) (Variables|BaseSetup|GetHologramVisible))
            (if (Variables|BaseSetup|GetExcludefromMediumHLOD)
              (HLOD|SetExcludedFromHLODLevel _hologramcountdown "HLOD0" true)
              (HLOD|SetExcludedFromHLODLevel _hologramskybeam_2 "HLOD0" true)
              (else
                (HLOD|SetExcludedFromHLODLevel _hologramcountdown "HLOD0" false)
                (HLOD|SetExcludedFromHLODLevel _hologramcountdown "HLOD1" false)
                (HLOD|SetExcludedFromHLODLevel _hologramskybeam_2 "HLOD0" false)
                (HLOD|SetExcludedFromHLODLevel _hologramskybeam_2 "HLOD1" false)))
            (else
              (HLOD|SetExcludedFromHLODLevel _hologramcountdown "HLOD0" true)
              (HLOD|SetExcludedFromHLODLevel _hologramcountdown "HLOD1" true)
              (HLOD|SetExcludedFromHLODLevel _hologramskybeam_2 "HLOD0" true)
              (HLOD|SetExcludedFromHLODLevel _hologramskybeam_2 "HLOD1" true))))))
    (:then_5
      (bind _returnvalue_2 (Rendering|Material|CreateDynamicMaterialInstance (Variables|Default|GetCountdownScreen) 0 0 "None" :!class /Script/Engine.PrimitiveComponent))
      (Variables|Default|SetMI_CountdownScreen _returnvalue_2)
      (bind _returnvalue_3 (Rendering|Material|CreateDynamicMaterialInstance (Variables|Default|GetHologramCountdown) 0 0 "None" :!class /Script/Engine.PrimitiveComponent))
      (Variables|Default|SetMI_Countdown_Hologram _returnvalue_3)
      (bind _returnvalue_4 (Rendering|Material|CreateDynamicMaterialInstance (Variables|Default|GetHologramSkybeam) 0 0 "None" :!class /Script/Engine.PrimitiveComponent))
      (Variables|Default|SetMI_Skybeam _returnvalue_4)
      (Rendering|Material|SetVectorParameterValue (Variables|Default|GetMI_CountdownScreen) "Color" (Math|Conversions|ToLinearColor(Color) (Variables|Default|GetColor)) :!class /Script/Engine.MaterialInstanceDynamic)
      (Rendering|Material|SetScalarParameterValue (Variables|Default|GetMI_CountdownScreen) "NumberofDigits" (Math|Conversions|ToFloat(Integer) (Variables|BaseSetup|GetNumberofDigits)) :!class /Script/Engine.MaterialInstanceDynamic)
      (Rendering|Material|SetScalarParameterValue (Variables|Default|GetMI_CountdownScreen) "u_NumberPaddingScale" (Variables|Clock|GetUV_Padding_x) :!class /Script/Engine.MaterialInstanceDynamic)
      (Rendering|Material|SetScalarParameterValue (Variables|Default|GetMI_CountdownScreen) "v_NumberPaddingScale" (Variables|Clock|GetUV_Padding_y) :!class /Script/Engine.MaterialInstanceDynamic)
      (Rendering|Material|SetScalarParameterValue (Variables|Default|GetMI_CountdownScreen) "u_NumberPaddingOffset" (Variables|Clock|GetUV_Offset_x) :!class /Script/Engine.MaterialInstanceDynamic)
      (Rendering|Material|SetScalarParameterValue (Variables|Default|GetMI_CountdownScreen) "v_NumberPaddingOffset" (Variables|Clock|GetUV_Offset_y) :!class /Script/Engine.MaterialInstanceDynamic)
      (Rendering|Material|SetScalarParameterValue (Variables|Default|GetMI_CountdownScreen) "Vignette" (Variables|Hologram|GetHologramVignette) :!class /Script/Engine.MaterialInstanceDynamic)
      (Rendering|Material|SetScalarParameterValue (Variables|Default|GetMI_Countdown_Hologram) "DistanceFadeLength" (Variables|Hologram|GetDistanceFadeLength) :!class /Script/Engine.MaterialInstanceDynamic)
      (Rendering|Material|SetScalarParameterValue (Variables|Default|GetDontUSEMI_Countdown_Hologram_HLOD) "DistanceFadeOffset" (Variables|Hologram|GetDistanceFadeOffset) :!class /Script/Engine.MaterialInstanceDynamic)
      (Rendering|Material|SetTextureParameterValue (Variables|Default|GetMI_CountdownScreen) "Background" (Variables|Default|GetBackgroundTexture))
      (Rendering|Material|SetTextureParameterValue (Variables|Default|GetMI_CountdownScreen) "PostCountdownImage" (Variables|Default|GetPostCountdownTexture))
      (Rendering|Material|SetTextureParameterValue (Variables|Default|GetMI_CountdownScreen) "DHMS" (Variables|Default|GetDHMSTexture))
      (Rendering|Material|SetScalarParameterValue (Variables|Default|GetMI_Countdown_Hologram) "ScaleX" (Variables|Hologram|GetScaleHologram_x) :!class /Script/Engine.MaterialInstanceDynamic)
      (Rendering|Material|SetScalarParameterValue (Variables|Default|GetMI_Countdown_Hologram) "ScaleY" (Variables|Hologram|GetScaleHologram_y) :!class /Script/Engine.MaterialInstanceDynamic)
      (Rendering|Material|SetScalarParameterValue (Variables|Default|GetMI_Skybeam) "Width2" (Variables|Hologram|GetScaleHologram_x) :!class /Script/Engine.MaterialInstanceDynamic))
    (:then_6
      ;[1 1 1] Material HLOD override
      (Rendering|Material|SetMaterial (Variables|Default|GetHologramCountdown) 1 (Variables|MaterialOverrides|GetHLOD%20Override_Hologram))
      (Rendering|Material|SetMaterial (Variables|Default|GetHologramSkybeam) 1 (Variables|MaterialOverrides|GetHLOD%20Override_SkyBeam)))
    (:then_7
      ;[1 1 1] Boundscale
      (Rendering|SetBoundsScale (Variables|Default|GetHologramCountdown) (/ (Math|Float|Max(Float) (Variables|Hologram|GetScaleHologram_x) (Variables|Hologram|GetScaleHologram_y)) 1e+02)))
    (:then_8
      ;[1 1 1] Debug different States
      (if (Variables|Debug|GetDebug_Phase2)
        (Rendering|Material|SetScalarParameterValue "/Game/Athena/Events/Generic/CountdownTimer/Materials/MPC_Countdown.MPC_Countdown" "ShowCountdown" 1.0 :!class /Script/Engine.KismetMaterialLibrary)
        (Rendering|Material|SetScalarParameterValue "/Game/Athena/Events/Generic/CountdownTimer/Materials/MPC_Countdown.MPC_Countdown" "ShowCountdownPhase2" 1.0 :!class /Script/Engine.KismetMaterialLibrary)
        (if (Variables|Debug|GetDebug_Phase3)
          (Rendering|Material|SetScalarParameterValue "/Game/Athena/Events/Generic/CountdownTimer/Materials/MPC_Countdown.MPC_Countdown" "ShowCountdownPhase3" 1.0 :!class /Script/Engine.KismetMaterialLibrary)
          (else
            (Rendering|Material|SetScalarParameterValue "/Game/Athena/Events/Generic/CountdownTimer/Materials/MPC_Countdown.MPC_Countdown" "ShowCountdownPhase3" :!class /Script/Engine.KismetMaterialLibrary)))))))
